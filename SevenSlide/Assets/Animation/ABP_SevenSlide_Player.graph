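# Fortnite player anim blueprint for SevenSlide, implementing the SecondMovementMode_FinalPoseOverride layer.
# A Locomotion state machine plays directional or slowmo intros into a held base pose, then adds turn or lean blendspace additives.
# bSlider_IsEquipped selects between that full-body pose and one layering the incoming upper body for other weapons, both with masked noise.
# Foot/leg IK and additive head-to-pelvis rotations follow, then a weighted thigh noise layer finishes.
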AnimBlueprint /SevenSlide/Assets/Animation/ABP_SevenSlide_Player
target_skeleton=&Fortnite_M_Avg_Player_Skeleton
graphs [
 AnimGraph {nodes [Output Pose()]}
 EventGraph {links [], nodes []}
 SecondMovementMode_FinalPoseOverride {nodes [
   n1 = Layered blend per bone(
    BasePose=Use cached pose 'FinalAnim_Cache'()
    BlendPoses_0=Apply Additive(
     Base=Use cached pose 'FinalAnim_Cache'()
     Additive="Sequence Player (additive) 'SevenSlide_Noise_Additive'"() {sequence=&/SevenSlide_Noise_Additive}
    )
    BlendWeights_0=Get EnableTurnOrLeanBS() -> (EnableTurnOrLeanBS:float)
   ) {bMeshSpaceRotationBlend=true, layerSetup [{branchFilters [{boneName=thigh_l}, {boneName=thigh_r}]}]}
   Output Pose(Result=@n1) {layerGroup=SecondMovementMode}
   AnimGraphNode_BlendSpaceGraph_1 = Blendspace 'TurnBS'!(X:float=Get TurnAmount() -> (TurnAmount:double)) -> (Pose:Pose Link Structure) {is=title AnimGraphNode_BlendSpaceGraph}
   n2 = Use cached pose 'Locomotion Cache'()
   n3 = Blendspace Player 'SevenSlide_Lean_Additive_BS'!(X=Get LeanAmount() -> (LeanAmount:double)) -> (Pose) {blendSpace=&/SevenSlide_Lean_Additive_BS}
   n4 = Use cached pose 'Locomotion Cache'()
   Save cached pose 'Locomotion+BS Cache'(Pose=Blend Poses by bool(
     BlendPose_0=Apply Mesh Space Additive!(Base:Pose Link Structure=@n2, Additive:Pose Link Structure=@AnimGraphNode_BlendSpaceGraph_1, Alpha:float=1) -> (Pose:Pose Link Structure) {is=title AnimGraphNode_ApplyMeshSpaceAdditive}
     BlendPose_1=Apply Mesh Space Additive(Base=@n4, Additive=@n3, Alpha=1)
     BlendTime_0=0.8
     BlendTime_1=1.4
     bActiveValue=Get bUseTurnBS() -> (bUseTurnBS:bool)
    ) {blendType=QuadraticInOut}) {cacheName=Locomotion+BS Cache}
   Save cached pose 'Input Pose Cache'(Pose=Input Pose() {name=InFinalPose}) {cacheName=Input Pose Cache}
   AnimGraphNode_StateMachine_1 = Locomotion() {is=title AnimGraphNode_StateMachine}
   Save cached pose 'Locomotion Cache'(Pose=@AnimGraphNode_StateMachine_1) {cacheName=Locomotion Cache}
   Save cached pose 'Final_Loco_Cache'(Pose=Layered blend per bone(BasePose=Use cached pose 'Locomotion Cache'(), BlendPoses_0=Use cached pose 'Locomotion+BS Cache'(), BlendWeights_0=Get EnableTurnOrLeanBS()) {layerSetup [{branchFilters [{boneName=root}]}]}) {cacheName=Final_Loco_Cache}
   Save cached pose 'MiscWeapon_Cache'(Pose=Layered blend per bone(BasePose=Use cached pose 'Final_Loco_Cache'(), BlendPoses_0=Use cached pose 'Input Pose Cache'()) {bMeshSpaceRotationBlend=true, layerSetup [{branchFilters [{blendDepth=3, boneName=Spine_04}]}]}) {cacheName=MiscWeapon_Cache}
   n5 = Get SevenSlide_LegIK_Alpha() -> (SevenSlide_LegIK_Alpha:float)
   n6 = Layered blend per bone(
    BasePose=Use cached pose 'Final_Loco_Cache'()
    BlendPoses_0=Apply Additive(
     Base=Use cached pose 'Final_Loco_Cache'()
     Additive="Sequence Player (additive) 'SevenSlide_Noise_Additive'"() {sequence=&SevenSlide_Noise_Additive}
    )
   ) {
    bMeshSpaceRotationBlend=true
    layerSetup [{branchFilters [{boneName=root}, {blendDepth=-1, boneName=thigh_l}, {blendDepth=-1, boneName=thigh_r}]}]
   }
   n7 = Layered blend per bone(
    BasePose=Use cached pose 'MiscWeapon_Cache'()
    BlendPoses_0=Apply Additive(
     Base=Use cached pose 'MiscWeapon_Cache'()
     Additive="Sequence Player (additive) 'SevenSlide_Noise_Additive'"() {sequence=&SevenSlide_Noise_Additive}
    )
   ) {
    bMeshSpaceRotationBlend=true
    layerSetup [{branchFilters [
       {boneName=root}
       {blendDepth=-1, boneName=clavicle_r}
       {blendDepth=-1, boneName=clavicle_l}
       {blendDepth=-1, boneName=thigh_l}
       {blendDepth=-1, boneName=thigh_r}
      ]}]
   }
   n8 = Get bSlider_IsEquipped() -> (bSlider_IsEquipped:bool)
   n9 = Set Foot IK Targets!(
    ComponentPose:Component Space Pose Link Structure=Local To Component(LocalPose=Blend Poses by bool(BlendPose_0=@n6, BlendPose_1=@n7, BlendTime_0=0.25, BlendTime_1=0.25, bActiveValue=@n8))
    Alpha:float=@n5
   ) -> (Pose:Component Space Pose Link Structure) {
    is=title FortAnimGraphNode_SetIKTarget
    iKTraceLength=70
    maxFootHeightComponentSpace=33
    alphaBoolBlend {blendOption=Linear}
    footL {boneName=foot_l}
    footR {boneName=foot_r}
    iKFootL {boneName=ik_foot_l}
    iKFootR {boneName=ik_foot_r}
   }
   n10 = Leg IK!(ComponentPose:Component Space Pose Link Structure=@n9, Alpha:float=@n5, SoftPercentLength:float=1) -> (Pose:Component Space Pose Link Structure) {
    is=title AnimGraphNode_LegIK
    softPercentLength=0.95
    alphaBoolBlend {blendOption=Linear}
    legsDefinition [
     {
      bEnableKneeTwistCorrection=false
      bEnableRotationLimit=false
      footBoneForwardAxis=Y
      hingeRotationAxis=Z
      minRotationAngle=15
      numBonesInLimb=2
      twistOffsetCurveName=None
      fKFootBone {boneName=foot_l}
      iKFootBone {boneName=ik_foot_l}
     }
     {
      bEnableKneeTwistCorrection=false
      bEnableRotationLimit=false
      footBoneForwardAxis=Y
      hingeRotationAxis=Z
      minRotationAngle=15
      numBonesInLimb=2
      twistOffsetCurveName=None
      fKFootBone {boneName=foot_r}
      iKFootBone {boneName=ik_foot_r}
     }
    ]
   }
   n11 = "Transform (Modify) Bone - Bone: head"!(ComponentPose=@n10, Alpha=Get Head_InputRot_Alpha() -> (Head_InputRot_Alpha:float), Rotation=Get Rot_Head() -> (Rot_Head:Rotator)) -> (Pose) {rotationMode=BMM_Additive, boneToModify {boneName=head}}
   n12 = Get Rot_Neck01() -> (Rot_Neck01:Rotator)
   n13 = Get Head_InputRot_Alpha()
   n14 = "Transform (Modify) Bone - Bone: spine_04"!(
    ComponentPose="Transform (Modify) Bone - Bone: spine_05"!(
     ComponentPose="Transform (Modify) Bone - Bone: neck_01"!(ComponentPose=@n11, Alpha=@n13, Rotation=@n12) -> (Pose) {rotationMode=BMM_Additive, boneToModify {boneName=neck_01}}
     Alpha=Get Spine_InputRot_Alpha() -> (Spine_InputRot_Alpha:float)
     Rotation=Get Rot_Spine05() -> (Rot_Spine05:Rotator)
    ) -> (Pose) {rotationMode=BMM_Additive, boneToModify {boneName=spine_05}}
    Alpha=Get Spine_InputRot_Alpha()
    Rotation=Get Rot_Spine04() -> (Rot_Spine04:Rotator)
   ) -> (Pose) {rotationMode=BMM_Additive, boneToModify {boneName=spine_04}}
   n15 = Get Rot_Spine03() -> (Rot_Spine03:Rotator)
   n16 = Get Spine_InputRot_Alpha()
   n17 = "Transform (Modify) Bone - Bone: Spine_01"!(
    ComponentPose="Transform (Modify) Bone - Bone: Spine_02"!(
     ComponentPose="Transform (Modify) Bone - Bone: Spine_03"!(ComponentPose=@n14, Alpha=@n16, Rotation=@n15) -> (Pose) {rotationMode=BMM_Additive, boneToModify {boneName=Spine_03}}
     Alpha=Get Pelvis_InputRot_Alpha() -> (Pelvis_InputRot_Alpha:float)
     Rotation=Get Rot_Spine02() -> (Rot_Spine02:Rotator)
    ) -> (Pose) {rotationMode=BMM_Additive, boneToModify {boneName=Spine_02}}
    Alpha=Get Pelvis_InputRot_Alpha()
    Rotation=Get Rot_Spine01() -> (Rot_Spine01:Rotator)
   ) -> (Pose) {rotationMode=BMM_Additive, boneToModify {boneName=Spine_01}}
   n18 = Get Rot_Pelvis() -> (Rot_Pelvis:Rotator)
   n19 = Get Pelvis_InputRot_Alpha()
   Save cached pose 'FinalAnim_Cache'(Pose=Component To Local(ComponentPose="Transform (Modify) Bone - Bone: pelvis"!(ComponentPose=@n17, Alpha=@n19, Rotation=@n18) -> (Pose) {rotationMode=BMM_Additive, boneToModify {boneName=pelvis}})) {cacheName=FinalAnim_Cache}
  ]}
 TurnBS {path=SecondMovementMode_FinalPoseOverride.AnimGraphNode_BlendSpaceGraph_1.TurnBS, links [], nodes []}
 NewSample {
  path=SecondMovementMode_FinalPoseOverride.AnimGraphNode_BlendSpaceGraph_1.TurnBS.NewSample
  nodes [Output Animation Pose!(Result:Pose Link Structure="Sequence Player (additive) 'SevenSlide_Lean_Additive_Center'"() {method=Graph, sequence=&/SevenSlide_Lean_Additive_Center}) {is=title AnimGraphNode_BlendSpaceSampleResult}]
 }
 NewSample1 {
  path=SecondMovementMode_FinalPoseOverride.AnimGraphNode_BlendSpaceGraph_1.TurnBS.NewSample1
  nodes [Output Animation Pose(Result="Sequence Player (additive) 'SevenSlide_Lean_Additive_Back'"() {method=Graph, sequence=&/SevenSlide_Lean_Additive_Back})]
 }
 NewSample2 {
  path=SecondMovementMode_FinalPoseOverride.AnimGraphNode_BlendSpaceGraph_1.TurnBS.NewSample2
  nodes [Output Animation Pose(Result="Sequence Player (additive) 'SevenSlide_Lean_Additive_Back'"() {method=Graph, sequence=&SevenSlide_Lean_Additive_Back})]
 }
 NewSample3 {
  path=SecondMovementMode_FinalPoseOverride.AnimGraphNode_BlendSpaceGraph_1.TurnBS.NewSample3
  nodes [Output Animation Pose(Result="Sequence Player (additive) 'SevenSlide_Lean_Additive_Right'"() {method=Graph, sequence=&/SevenSlide_Lean_Additive_Right})]
 }
 NewSample4 {
  path=SecondMovementMode_FinalPoseOverride.AnimGraphNode_BlendSpaceGraph_1.TurnBS.NewSample4
  nodes [Output Animation Pose(Result="Sequence Player (additive) 'SevenSlide_Lean_Additive_Left'"() {method=Graph, sequence=&/SevenSlide_Lean_Additive_Left})]
 }
 Locomotion {
  entry=Intro
  path=SecondMovementMode_FinalPoseOverride.AnimGraphNode_StateMachine_1.Locomotion
  ids=id
  nodes [
   AnimStateNode_0 = Intro() {is=title AnimStateNode}
   AnimStateNode_1 = Loop() {is=title AnimStateNode}
   AnimStateTransitionNode_0 = AnimStateTransitionNode() {
    bAllowInertializationForSelfTransitions=true
    bAutomaticRuleBasedOnSequencePlayerInState=true
    blendMode=QuadraticInOut
    crossfadeDuration=0.4
   }
  ]
  links [{
    dst=Loop
    rule=SecondMovementMode_FinalPoseOverride.AnimGraphNode_StateMachine_1.Locomotion.AnimStateTransitionNode_0.Transition
    src=Intro
    via=AnimStateTransitionNode_0
   }]
 }
 Intro {
  path=SecondMovementMode_FinalPoseOverride.AnimGraphNode_StateMachine_1.Locomotion.AnimStateNode_0.Intro
  nodes [
   AnimGraphNode_StateMachine_1 = Intros() {is=title AnimGraphNode_StateMachine, bAllowConduitEntryStates=true}
   Output Animation Pose(Result=@AnimGraphNode_StateMachine_1) {is=title AnimGraphNode_StateResult}
  ]
 }
 Intros {
  entry=Conduit
  path=SecondMovementMode_FinalPoseOverride.AnimGraphNode_StateMachine_1.Locomotion.AnimStateNode_0.Intro.AnimGraphNode_StateMachine_1.Intros
  ids=id
  nodes [
   AnimStateConduitNode_0 = Conduit()
   AnimStateNode_1 = Fw_Slowmo() {
    is=title AnimStateNode
    stateEntered {guid=#1}
    stateFullyBlended {guid=#2}
    stateLeft {guid=#3}
   }
   AnimStateNode_2 = L_Slowmo() {is=title AnimStateNode}
   AnimStateNode_3 = R_Slowmo() {is=title AnimStateNode}
   AnimStateNode_4 = Bw_Slowmo() {is=title AnimStateNode}
   AnimStateNode_5 = Fw() {
    is=title AnimStateNode
    stateEntered {guid=#4}
    stateFullyBlended {guid=#5}
    stateLeft {guid=#6}
   }
   AnimStateNode_6 = Bw() {is=title AnimStateNode}
   AnimStateNode_7 = L() {is=title AnimStateNode}
   AnimStateNode_8 = R() {is=title AnimStateNode}
   AnimStateTransitionNode_0 = AnimStateTransitionNode() {
    bAllowInertializationForSelfTransitions=true
    crossfadeDuration=0.4
    priorityOrder=2
    transitionEnd {guid=#7}
    transitionInterrupt {guid=#8}
    transitionStart {guid=#9}
   }
   AnimStateTransitionNode_1 = AnimStateTransitionNode() {bAllowInertializationForSelfTransitions=true, crossfadeDuration=0.4, priorityOrder=2}
   AnimStateTransitionNode_2 = AnimStateTransitionNode() {bAllowInertializationForSelfTransitions=true, crossfadeDuration=0.4, priorityOrder=2}
   AnimStateTransitionNode_3 = AnimStateTransitionNode() {bAllowInertializationForSelfTransitions=true, crossfadeDuration=0.4, priorityOrder=2}
   AnimStateTransitionNode_4 = AnimStateTransitionNode() {
    bAllowInertializationForSelfTransitions=true
    crossfadeDuration=0.4
    transitionEnd {guid=#10}
    transitionInterrupt {guid=#11}
    transitionStart {guid=#12}
   }
   AnimStateTransitionNode_5 = AnimStateTransitionNode() {bAllowInertializationForSelfTransitions=true, crossfadeDuration=0.4}
   AnimStateTransitionNode_6 = AnimStateTransitionNode() {bAllowInertializationForSelfTransitions=true, crossfadeDuration=0.4}
   AnimStateTransitionNode_7 = AnimStateTransitionNode() {bAllowInertializationForSelfTransitions=true, crossfadeDuration=0.4}
  ]
  links [
   {
    dst=Bw
    rule=SecondMovementMode_FinalPoseOverride.AnimGraphNode_StateMachine_1.Locomotion.AnimStateNode_0.Intro.AnimGraphNode_StateMachine_1.Intros.AnimStateTransitionNode_1.Transition
    src=Conduit
    via=AnimStateTransitionNode_1
   }
   {
    dst=Bw_Slowmo
    rule=SecondMovementMode_FinalPoseOverride.AnimGraphNode_StateMachine_1.Locomotion.AnimStateNode_0.Intro.AnimGraphNode_StateMachine_1.Intros.AnimStateTransitionNode_5.Transition
    src=Conduit
    via=AnimStateTransitionNode_5
   }
   {
    dst=Fw
    rule=SecondMovementMode_FinalPoseOverride.AnimGraphNode_StateMachine_1.Locomotion.AnimStateNode_0.Intro.AnimGraphNode_StateMachine_1.Intros.AnimStateTransitionNode_0.Transition
    src=Conduit
    via=AnimStateTransitionNode_0
   }
   {
    dst=Fw_Slowmo
    rule=SecondMovementMode_FinalPoseOverride.AnimGraphNode_StateMachine_1.Locomotion.AnimStateNode_0.Intro.AnimGraphNode_StateMachine_1.Intros.AnimStateTransitionNode_4.Transition
    src=Conduit
    via=AnimStateTransitionNode_4
   }
   {
    dst=L
    rule=SecondMovementMode_FinalPoseOverride.AnimGraphNode_StateMachine_1.Locomotion.AnimStateNode_0.Intro.AnimGraphNode_StateMachine_1.Intros.AnimStateTransitionNode_2.Transition
    src=Conduit
    via=AnimStateTransitionNode_2
   }
   {
    dst=L_Slowmo
    rule=SecondMovementMode_FinalPoseOverride.AnimGraphNode_StateMachine_1.Locomotion.AnimStateNode_0.Intro.AnimGraphNode_StateMachine_1.Intros.AnimStateTransitionNode_6.Transition
    src=Conduit
    via=AnimStateTransitionNode_6
   }
   {
    dst=R
    rule=SecondMovementMode_FinalPoseOverride.AnimGraphNode_StateMachine_1.Locomotion.AnimStateNode_0.Intro.AnimGraphNode_StateMachine_1.Intros.AnimStateTransitionNode_3.Transition
    src=Conduit
    via=AnimStateTransitionNode_3
   }
   {
    dst=R_Slowmo
    rule=SecondMovementMode_FinalPoseOverride.AnimGraphNode_StateMachine_1.Locomotion.AnimStateNode_0.Intro.AnimGraphNode_StateMachine_1.Intros.AnimStateTransitionNode_7.Transition
    src=Conduit
    via=AnimStateTransitionNode_7
   }
  ]
 }
 Conduit {
  path=SecondMovementMode_FinalPoseOverride.AnimGraphNode_StateMachine_1.Locomotion.AnimStateNode_0.Intro.AnimGraphNode_StateMachine_1.Intros.AnimStateConduitNode_0.Conduit
  nodes [Result(bCanEnterTransition=true)]
 }
 Fw_Slowmo {
  path=SecondMovementMode_FinalPoseOverride.AnimGraphNode_StateMachine_1.Locomotion.AnimStateNode_0.Intro.AnimGraphNode_StateMachine_1.Intros.AnimStateNode_1.Fw_Slowmo
  nodes [Output Animation Pose(Result=Sequence Player 'SevenSlide_Intro_Small'() {bLoopAnimation=false, sequence=&/SevenSlide_Intro_Small})]
 }
 L_Slowmo {
  path=SecondMovementMode_FinalPoseOverride.AnimGraphNode_StateMachine_1.Locomotion.AnimStateNode_0.Intro.AnimGraphNode_StateMachine_1.Intros.AnimStateNode_2.L_Slowmo
  nodes [Output Animation Pose(Result=Sequence Player 'SevenSlide_Intro_L_Small'() {bLoopAnimation=false, sequence=&/SevenSlide_Intro_L_Small})]
 }
 R_Slowmo {
  path=SecondMovementMode_FinalPoseOverride.AnimGraphNode_StateMachine_1.Locomotion.AnimStateNode_0.Intro.AnimGraphNode_StateMachine_1.Intros.AnimStateNode_3.R_Slowmo
  nodes [Output Animation Pose(Result=Sequence Player 'SevenSlide_Intro_R_Small'() {bLoopAnimation=false, sequence=&/SevenSlide_Intro_R_Small})]
 }
 Bw_Slowmo {
  path=SecondMovementMode_FinalPoseOverride.AnimGraphNode_StateMachine_1.Locomotion.AnimStateNode_0.Intro.AnimGraphNode_StateMachine_1.Intros.AnimStateNode_4.Bw_Slowmo
  nodes [Output Animation Pose(Result=Sequence Player 'SevenSlide_Intro_Back_Small'() {bLoopAnimation=false, sequence=&/SevenSlide_Intro_Back_Small})]
 }
 Fw {
  path=SecondMovementMode_FinalPoseOverride.AnimGraphNode_StateMachine_1.Locomotion.AnimStateNode_0.Intro.AnimGraphNode_StateMachine_1.Intros.AnimStateNode_5.Fw
  nodes [Output Animation Pose(Result=Sequence Player 'SevenSlide_Intro'() {bLoopAnimation=false, sequence=&/SevenSlide_Intro})]
 }
 Bw {
  path=SecondMovementMode_FinalPoseOverride.AnimGraphNode_StateMachine_1.Locomotion.AnimStateNode_0.Intro.AnimGraphNode_StateMachine_1.Intros.AnimStateNode_6.Bw
  nodes [Output Animation Pose(Result=Sequence Player 'SevenSlide_Intro_Back'() {bLoopAnimation=false, sequence=&/SevenSlide_Intro_Back})]
 }
 L {
  path=SecondMovementMode_FinalPoseOverride.AnimGraphNode_StateMachine_1.Locomotion.AnimStateNode_0.Intro.AnimGraphNode_StateMachine_1.Intros.AnimStateNode_7.L
  nodes [Output Animation Pose(Result=Sequence Player 'SevenSlide_Intro_L'() {bLoopAnimation=false, sequence=&/SevenSlide_Intro_L})]
 }
 R {
  path=SecondMovementMode_FinalPoseOverride.AnimGraphNode_StateMachine_1.Locomotion.AnimStateNode_0.Intro.AnimGraphNode_StateMachine_1.Intros.AnimStateNode_8.R
  nodes [Output Animation Pose(Result=Sequence Player 'SevenSlide_Intro_R'() {bLoopAnimation=false, sequence=&/SevenSlide_Intro_R})]
 }
 Transition {
  path=SecondMovementMode_FinalPoseOverride.AnimGraphNode_StateMachine_1.Locomotion.AnimStateNode_0.Intro.AnimGraphNode_StateMachine_1.Intros.AnimStateTransitionNode_0.Transition
  nodes [Result(bCanEnterTransition=Get bTransitionRule_ConduitToFw() -> (bTransitionRule_ConduitToFw:bool))]
 }
 Transition {
  path=SecondMovementMode_FinalPoseOverride.AnimGraphNode_StateMachine_1.Locomotion.AnimStateNode_0.Intro.AnimGraphNode_StateMachine_1.Intros.AnimStateTransitionNode_1.Transition
  nodes [Result(bCanEnterTransition=Get bTransitionRule_ConduitToBw() -> (bTransitionRule_ConduitToBw:bool))]
 }
 Transition {
  path=SecondMovementMode_FinalPoseOverride.AnimGraphNode_StateMachine_1.Locomotion.AnimStateNode_0.Intro.AnimGraphNode_StateMachine_1.Intros.AnimStateTransitionNode_2.Transition
  nodes [Result(bCanEnterTransition=Get bTransitionRule_ConduitToL() -> (bTransitionRule_ConduitToL:bool))]
 }
 Transition {
  path=SecondMovementMode_FinalPoseOverride.AnimGraphNode_StateMachine_1.Locomotion.AnimStateNode_0.Intro.AnimGraphNode_StateMachine_1.Intros.AnimStateTransitionNode_3.Transition
  nodes [Result(bCanEnterTransition=Get bTransitionRule_ConduitToR() -> (bTransitionRule_ConduitToR:bool))]
 }
 Transition {
  path=SecondMovementMode_FinalPoseOverride.AnimGraphNode_StateMachine_1.Locomotion.AnimStateNode_0.Intro.AnimGraphNode_StateMachine_1.Intros.AnimStateTransitionNode_4.Transition
  nodes [Result(bCanEnterTransition=Get bTransitionRule_ConduitToFwSlowmo() -> (bTransitionRule_ConduitToFwSlowmo:bool))]
 }
 Transition {
  path=SecondMovementMode_FinalPoseOverride.AnimGraphNode_StateMachine_1.Locomotion.AnimStateNode_0.Intro.AnimGraphNode_StateMachine_1.Intros.AnimStateTransitionNode_5.Transition
  nodes [Result(bCanEnterTransition=Get bTransitionRule_ConduitToBwSlowmo() -> (bTransitionRule_ConduitToBwSlowmo:bool))]
 }
 Transition {
  path=SecondMovementMode_FinalPoseOverride.AnimGraphNode_StateMachine_1.Locomotion.AnimStateNode_0.Intro.AnimGraphNode_StateMachine_1.Intros.AnimStateTransitionNode_6.Transition
  nodes [Result(bCanEnterTransition=Get bTransitionRule_ConduitToLSlowmo() -> (bTransitionRule_ConduitToLSlowmo:bool))]
 }
 Transition {
  path=SecondMovementMode_FinalPoseOverride.AnimGraphNode_StateMachine_1.Locomotion.AnimStateNode_0.Intro.AnimGraphNode_StateMachine_1.Intros.AnimStateTransitionNode_7.Transition
  nodes [Result(bCanEnterTransition=Get bTransitionRule_ConduitToRSlowmo() -> (bTransitionRule_ConduitToRSlowmo:bool))]
 }
 Loop {
  path=SecondMovementMode_FinalPoseOverride.AnimGraphNode_StateMachine_1.Locomotion.AnimStateNode_1.Loop
  nodes [Output Animation Pose(Result=Modify Curve!(
     SourcePose=Sequence Player 'SevenSlide_BasePose'() {bLoopAnimation=false, sequence=&/SevenSlide_BasePose}
     CurveValues_0=1
     CurveValues_1:float=1
     Alpha=1
    ) -> (Pose) {curveNames [SevenSlide_LegIK_Alpha, Enable_TurnOrLeanBS], curveValues [0,0]})]
 }
 Transition {path=SecondMovementMode_FinalPoseOverride.AnimGraphNode_StateMachine_1.Locomotion.AnimStateTransitionNode_0.Transition, nodes [Result()]}
]
variables []
