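# Stereo looping MetaSound source for the SevenSlide overheat cool-down, pairing a random-get cool-down wave with a low-fuel warning layer.
# FuelLevel is range-mapped and interpolated to drive both pitch, via semitone conversion plus PitchShift, and cool-down gain.
# Delayed MSP_Fade envelopes shape each layer before a two-input mixer; the MasterGain input is unconnected.
# Control buses modulate volume and low-pass, and concurrency allows one instance per owner.
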MetaSoundSource /SevenSlide/Assets/Audio/MSS_SevenSlide_CoolDown_Loop
inputs [
 FuelLevel {type=Float}
 MasterGain {type=Float}
 PitchShift {type=Float}
 Sounds {type=WaveAsset:Array}
 StartingValue {type=Float}
 UE.Source.OnPlay {type=Trigger}
]
interfaces [UE.OutputFormat.Stereo:1.0, UE.Source:1.1]
outputs [
 "UE.OutputFormat.Stereo.Audio:0" {type=Audio}
 "UE.OutputFormat.Stereo.Audio:1" {type=Audio}
]
pages [Default {
  input_defaults {
   FuelLevel=100
   MasterGain=1
   PitchShift=0
   StartingValue=1
   UE.Source.OnPlay=null
   Sounds [/SevenSlide/Assets/Audio/Wavs/SevenSlide_Overheat_CoolDown_Loop]
  }
  nodes [
   n1 = MapRange.MapRange:Float(In=^FuelLevel) {also {defaults {In Range A=100, In Range B=0.5}}}
   TriggerOnValueChange_0 = UE.Trigger On Value Change:Float(Value=@n1.Out Value)
   n2 = UE.InterpTo:Audio(Target=Value.Value:Float(Init Value=^StartingValue, Reset=^UE.Source.OnPlay, Target Value=@n1.Out Value, Set=@TriggerOnValueChange_0.Trigger).Output Value) {also {defaults {Interp Time=0.5}}}
   FrequencyMultipliertoSemitone_0 = UE.Frequency Multiplier to Semitone:Float(Frequency Multiplier=@n2.Value)
   ArrayRandomGet_0 = Array.Random Get:WaveAsset:Array(In Array=^Sounds, Next=^UE.Source.OnPlay)
   WavePlayer_1 = UE.Wave Player:Stereo(
    Pitch Shift=UE.Add:Float(AdditionalOperands=^PitchShift, PrimaryOperand=@FrequencyMultipliertoSemitone_0.Semitones).Out
    Play=@ArrayRandomGet_0.On Next
    Wave Asset=@ArrayRandomGet_0.Value
   ) {also {defaults {Loop="<MetaSoundBuilderResult.SUCCEEDED: 0>"}, version [1,1]}}
   TriggerDelay_1 = UE.Trigger Delay(In=@WavePlayer_1.On Play)
   n3 = UE.Multiply:Float(
    PrimaryOperand=@n2.Value
    AdditionalOperands=&MSP_Fade(In=@TriggerDelay_1.Out) {also {defaults {Fade Time=4, bInvertOutput="<MetaSoundBuilderResult.SUCCEEDED: 0>"}}}.Value Out
   )
   TriggerDelay_2 = UE.Trigger Delay(In=^UE.Source.OnPlay)
   WavePlayer_0 = UE.Wave Player:Mono(Play=@TriggerDelay_2.Out) {also {defaults {Loop="<MetaSoundBuilderResult.SUCCEEDED: 0>", Wave Asset=/SevenSlide/Assets/Audio/Wavs/SevenSlide_LowFuel_Warning}, version [1,1]}}
   TriggerDelay_0 = UE.Trigger Delay(In=@WavePlayer_0.On Play)
   AudioMixerAudioMixerStereo2_0 = "AudioMixer.Audio Mixer (Stereo, 2)"(
    In 0 R=UE.Multiply:Audio by Float(AdditionalOperands=@n3.Out, PrimaryOperand=@WavePlayer_1.Out Right).Out
    In 0 L=UE.Multiply:Audio by Float(AdditionalOperands=@n3.Out, PrimaryOperand=@WavePlayer_1.Out Left).Out
    Gain 1=UE.Multiply:Float(PrimaryOperand=&MSP_Fade(In=@TriggerDelay_0.Out) {also {defaults {Fade Time=2, bInvertOutput="<MetaSoundBuilderResult.SUCCEEDED: 0>"}}}.Value Out) {also {defaults {AdditionalOperands=0.45}}}.Out
    In 1 L=@WavePlayer_0.Out Mono
    In 1 R=@WavePlayer_0.Out Mono
   )
  ]
  links [
   {dst="@UE.OutputFormat.Stereo.Audio:0", src=AudioMixerAudioMixerStereo2_0.Out L}
   {dst="@UE.OutputFormat.Stereo.Audio:1", src=AudioMixerAudioMixerStereo2_0.Out R}
  ]
 }]
properties {
 assetImportData=&self:AssetImportData
 attenuationSettings=&/Att_SevenSlide_Movement2
 bLooping=true
 bOverrideConcurrency=true
 duration=10000
 maxDistance=12000
 numChannels=2
 outputFormat=Stereo
 sampleRate=48000
 soundClassObject=&Fort_WeaponMelee
 volume=0.5
 concurrencyOverrides {bLimitToOwner=true, maxCount=1, resolutionRule=PreventNew}
 modulationSettings {
  lowpassRouting=Union
  volumeRouting=Union
  lowpassModulationDestination {bEnableModulation=true, modulators [&/CB_SevenSlide_LowPass]}
  pitchModulationDestination {bEnableModulation=true, modulators [null]}
  volumeModulationDestination {bEnableModulation=true, modulators [&/CB_SevenSlide_Volume]}
 }
}
