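# One-shot stereo MetaSound Source that plays player foley step sounds for SevenSlide.
# On play, a shuffle picks one of twelve step wave variations for a stereo Wave Player, whose finish ends the source.
# It uses the SevenSlide dash attenuation and a shared player foley concurrency set, with SevenSlide control buses modulating volume and low-pass.
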MetaSoundSource /SevenSlide/Assets/Audio/MSS_SevenSlide_Foley_Step
inputs [
 Sounds {type=WaveAsset:Array}
 UE.Source.OnPlay {type=Trigger}
]
interfaces [UE.OutputFormat.Stereo:1.0, UE.Source.OneShot:1.0, UE.Source:1.1]
outputs [
 "UE.OutputFormat.Stereo.Audio:0" {type=Audio}
 "UE.OutputFormat.Stereo.Audio:1" {type=Audio}
 UE.Source.OneShot.OnFinished {type=Trigger}
]
pages [Default {
  input_defaults {
   UE.Source.OnPlay=null
   Sounds [
    /SevenSlide/Assets/Audio/Wavs/SevenSlide_Foley_Step_01
    /SevenSlide/Assets/Audio/Wavs/SevenSlide_Foley_Step_02
    /SevenSlide/Assets/Audio/Wavs/SevenSlide_Foley_Step_03
    /SevenSlide/Assets/Audio/Wavs/SevenSlide_Foley_Step_04
    /SevenSlide/Assets/Audio/Wavs/SevenSlide_Foley_Step_05
    /SevenSlide/Assets/Audio/Wavs/SevenSlide_Foley_Step_06
    /SevenSlide/Assets/Audio/Wavs/SevenSlide_Foley_Step_07
    /SevenSlide/Assets/Audio/Wavs/SevenSlide_Foley_Step_08
    /SevenSlide/Assets/Audio/Wavs/SevenSlide_Foley_Step_09
    /SevenSlide/Assets/Audio/Wavs/SevenSlide_Foley_Step_10
    /SevenSlide/Assets/Audio/Wavs/SevenSlide_Foley_Step_11
    /SevenSlide/Assets/Audio/Wavs/SevenSlide_Foley_Step_12
   ]
  }
  nodes [
   n1 = Array.Shuffle:WaveAsset:Array(In Array=^Sounds, Next=^UE.Source.OnPlay) {also {defaults {Enable Shared State="<MetaSoundBuilderResult.SUCCEEDED: 0>"}}}
   WavePlayer_0 = UE.Wave Player:Stereo(Play=@n1.On Next, Wave Asset=@n1.Value) {also {version [1,1]}}
  ]
  links [
   {dst="@UE.Source.OneShot.OnFinished", src=WavePlayer_0.On Finished}
   {dst="@UE.OutputFormat.Stereo.Audio:0", src=WavePlayer_0.Out Left}
   {dst="@UE.OutputFormat.Stereo.Audio:1", src=WavePlayer_0.Out Right}
  ]
 }]
properties {
 assetImportData=&self:AssetImportData
 attenuationSettings=&/ATT_SevenSlide_Dash
 maxDistance=4400
 numChannels=2
 outputFormat=Stereo
 priority=0.2
 sampleRate=48000
 soundClassObject=&Fort_Player_Foley
 volume=2
 concurrencyOverrides {maxCount=1, resolutionRule=PreventNew}
 concurrencySet [/Game/Sounds/Fort_Player/Clothing/Settings/Fort_Player_Foley_Global_MaxFour]
 modulationSettings {
  lowpassRouting=Union
  volumeRouting=Union
  lowpassModulationDestination {bEnableModulation=true, modulators [&/CB_SevenSlide_LowPass]}
  volumeModulationDestination {bEnableModulation=true, modulators [&/CB_SevenSlide_Volume]}
 }
}
