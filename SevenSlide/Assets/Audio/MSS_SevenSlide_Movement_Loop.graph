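# Stereo looping MetaSound source for SevenSlide movement audio, in the melee weapon sound class.
# On play, a random pick from the Sounds array, which holds one loop wave by default, feeds a stereo wave player straight to the outputs.
# The Start Burst Sound input is empty and unwired.
# Custom attenuation, a concurrency override and control-bus modulation of volume and low-pass shape playback outside the graph.
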MetaSoundSource /SevenSlide/Assets/Audio/MSS_SevenSlide_Movement_Loop
inputs [
 Sounds {type=WaveAsset:Array}
 Start Burst Sound {type=WaveAsset:Array}
 UE.Source.OnPlay {type=Trigger}
]
interfaces [UE.OutputFormat.Stereo:1.0, UE.Source:1.1]
outputs [
 "UE.OutputFormat.Stereo.Audio:0" {type=Audio}
 "UE.OutputFormat.Stereo.Audio:1" {type=Audio}
]
pages [Default {
  input_defaults {UE.Source.OnPlay=null, Sounds [/SevenSlide/Assets/Audio/Wavs/SevenSlide_Movement_Loop], Start Burst Sound []}
  nodes [
   ArrayRandomGet_0 = Array.Random Get:WaveAsset:Array(In Array=^Sounds, Next=^UE.Source.OnPlay)
   WavePlayer_0 = UE.Wave Player:Stereo(Play=@ArrayRandomGet_0.On Next, Wave Asset=@ArrayRandomGet_0.Value) {also {defaults {Loop="<MetaSoundBuilderResult.SUCCEEDED: 0>"}, version [1,1]}}
  ]
  links [{dst="@UE.OutputFormat.Stereo.Audio:0", src=WavePlayer_0.Out Left}, {dst="@UE.OutputFormat.Stereo.Audio:1", src=WavePlayer_0.Out Right}]
 }]
properties {
 assetImportData=&self:AssetImportData
 attenuationSettings=&/Att_SevenSlide_Movement2
 bLooping=true
 bOverrideConcurrency=true
 duration=10000
 maxDistance=12000
 numChannels=2
 outputFormat=Stereo
 sampleRate=48000
 soundClassObject=&Fort_WeaponMelee
 volume=1.2
 concurrencyOverrides {maxCount=4, volumeScaleAttackTime=0.1}
 modulationSettings {
  lowpassRouting=Union
  volumeRouting=Union
  lowpassModulationDestination {bEnableModulation=true, modulators [&/CB_SevenSlide_LowPass]}
  pitchModulationDestination {bEnableModulation=true, modulators [null]}
  volumeModulationDestination {bEnableModulation=true, modulators [&/CB_SevenSlide_Volume]}
 }
}
