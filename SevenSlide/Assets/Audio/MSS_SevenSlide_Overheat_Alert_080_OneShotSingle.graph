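# Mono one-shot MetaSound Source, named as the SevenSlide overheat alert, spatialized with the SevenSlide beeps attenuation and routed to the Fort_WeaponMelee sound class.
# On play, a single Wave Player plays the exposed Sound wave input straight to the mono output and signals finished when the wave ends.
# The Sound input has no default wave in this graph, and no pitch, gain or randomization processing is applied.
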MetaSoundSource /SevenSlide/Assets/Audio/MSS_SevenSlide_Overheat_Alert_080_OneShotSingle
inputs [
 Sound {type=WaveAsset}
 UE.Source.OnPlay {type=Trigger}
]
interfaces [UE.OutputFormat.Mono:1.0, UE.Source.OneShot:1.0, UE.Source:1.1]
outputs [
 "UE.OutputFormat.Mono.Audio:0" {type=Audio}
 UE.Source.OneShot.OnFinished {type=Trigger}
]
pages [Default {
  input_defaults {Sound=null, UE.Source.OnPlay=null}
  nodes [WavePlayer_0 = UE.Wave Player:Mono(Wave Asset=^Sound, Play=^UE.Source.OnPlay) {also {version [1,1]}}]
  links [{dst="@UE.Source.OneShot.OnFinished", src=WavePlayer_0.On Finished}, {dst="@UE.OutputFormat.Mono.Audio:0", src=WavePlayer_0.Out Mono}]
 }]
properties {
 assetImportData=&self:AssetImportData
 attenuationSettings=&/ATT_SevenSlide_Beeps
 maxDistance=2250
 soundClassObject=&Fort_WeaponMelee
}
