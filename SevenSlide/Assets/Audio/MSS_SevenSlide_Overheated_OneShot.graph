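# One-shot mono MetaSound source that plays the SevenSlide overheat power-down sound with distance attenuation, under a melee weapon sound class.
# On play, a random pick from two power-down waves feeds the main wave player, which also triggers a second player holding a steam-loop wave.
# Only the main player reaches the audio output, but either player finishing ends the source.
# The Sound input and Value output are unconnected.
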MetaSoundSource /SevenSlide/Assets/Audio/MSS_SevenSlide_Overheated_OneShot
inputs [
 Power Down Sound {type=WaveAsset:Array}
 Sound {type=WaveAsset}
 UE.Source.OnPlay {type=Trigger}
]
interfaces [UE.OutputFormat.Mono:1.0, UE.Source.OneShot:1.0, UE.Source:1.1]
outputs [
 "UE.OutputFormat.Mono.Audio:0" {type=Audio}
 UE.Source.OneShot.OnFinished {type=Trigger}
 Value {type=WaveAsset}
]
pages [Default {
  input_defaults {
   Sound=/SevenSlide/Assets/Audio/Wavs/SevenSlide_Overheat_02
   UE.Source.OnPlay=null
   Power Down Sound [/SevenSlide/Assets/Audio/Wavs/SevenSlide_Overheat_PowerDown_01, /SevenSlide/Assets/Audio/Wavs/SevenSlide_Overheat_PowerDown_02]
  }
  nodes [
   ArrayRandomGet_0 = Array.Random Get:WaveAsset:Array(In Array=^Power Down Sound, Next=^UE.Source.OnPlay)
   WavePlayer_0 = UE.Wave Player:Mono(Play=@ArrayRandomGet_0.On Next, Wave Asset=@ArrayRandomGet_0.Value) {also {version [1,1]}}
   WavePlayer_1 = UE.Wave Player:Mono(Play=@WavePlayer_0.On Play) {also {defaults {Wave Asset=/SevenSlide/Assets/Audio/Wavs/SevenSlide_SpeedBoost_SteamLoop}, version [1,1]}}
   TriggerAnyTriggerAny2_0 = "TriggerAny.Trigger Any (2)"(In 1=@WavePlayer_0.On Finished, In 0=@WavePlayer_1.On Finished)
  ]
  links [{dst="@UE.Source.OneShot.OnFinished", src=TriggerAnyTriggerAny2_0.Out}, {dst="@UE.OutputFormat.Mono.Audio:0", src=WavePlayer_0.Out Mono}]
 }]
properties {
 assetImportData=&self:AssetImportData
 attenuationSettings=&/ATT_SevenSlide_Start_Stop
 maxDistance=10500
 sampleRate=48000
 soundClassObject=&Fort_WeaponMelee
 volume=2
}
