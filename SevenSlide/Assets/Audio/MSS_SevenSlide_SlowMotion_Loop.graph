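# Stereo MetaSound source for the SevenSlide slow-motion loop, pairing an entry burst with a sustained loop sound.
# On play, two random-get nodes feed separate wave players: one from the single-entry loop array, one from four enter variations, pitch-shifted up.
# A stereo mixer sums both, with the burst at reduced gain.
# The source uses SevenSlide movement attenuation, the melee weapon sound class and a one-instance concurrency override.
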MetaSoundSource /SevenSlide/Assets/Audio/MSS_SevenSlide_SlowMotion_Loop
inputs [
 Sounds {type=WaveAsset:Array}
 Start Burst Sound {type=WaveAsset:Array}
 UE.Source.OnPlay {type=Trigger}
]
interfaces [UE.OutputFormat.Stereo:1.0, UE.Source:1.1]
outputs [
 "UE.OutputFormat.Stereo.Audio:0" {type=Audio}
 "UE.OutputFormat.Stereo.Audio:1" {type=Audio}
]
pages [Default {
  input_defaults {
   UE.Source.OnPlay=null
   Sounds [/SevenSlide/Assets/Audio/Wavs/SevenSlide_SlowMo_Loop_1P]
   Start Burst Sound [
    /SevenSlide/Assets/Audio/Wavs/SevenSlide_SlowMo_Enter_01
    /SevenSlide/Assets/Audio/Wavs/SevenSlide_SlowMo_Enter_02
    /SevenSlide/Assets/Audio/Wavs/SevenSlide_SlowMo_Enter_03
    /SevenSlide/Assets/Audio/Wavs/SevenSlide_SlowMo_Enter_04
   ]
  }
  nodes [
   ArrayRandomGet_0 = Array.Random Get:WaveAsset:Array(In Array=^Sounds, Next=^UE.Source.OnPlay)
   WavePlayer_0 = UE.Wave Player:Stereo(Play=@ArrayRandomGet_0.On Next, Wave Asset=@ArrayRandomGet_0.Value) {also {defaults {Loop="<MetaSoundBuilderResult.SUCCEEDED: 0>"}, version [1,1]}}
   ArrayRandomGet_1 = Array.Random Get:WaveAsset:Array(In Array=^Start Burst Sound, Next=^UE.Source.OnPlay) {also {defaults {Weights []}}}
   WavePlayer_1 = UE.Wave Player:Stereo(Play=@ArrayRandomGet_1.On Next, Wave Asset=@ArrayRandomGet_1.Value) {also {defaults {Pitch Shift=0.5}, version [1,1]}}
   AudioMixerAudioMixerStereo2_0 = "AudioMixer.Audio Mixer (Stereo, 2)"(In 0 L=@WavePlayer_0.Out Left, In 0 R=@WavePlayer_0.Out Right, In 1 L=@WavePlayer_1.Out Left, In 1 R=@WavePlayer_1.Out Right) {also {defaults {Gain 1=0.5}}}
  ]
  links [
   {dst="@UE.OutputFormat.Stereo.Audio:0", src=AudioMixerAudioMixerStereo2_0.Out L}
   {dst="@UE.OutputFormat.Stereo.Audio:1", src=AudioMixerAudioMixerStereo2_0.Out R}
  ]
 }]
properties {
 assetImportData=&self:AssetImportData
 attenuationSettings=&/Att_SevenSlide_Movement2
 bLooping=true
 bOverrideConcurrency=true
 duration=10000
 maxDistance=12000
 numChannels=2
 outputFormat=Stereo
 sampleRate=48000
 soundClassObject=&Fort_WeaponMelee
 volume=2
 concurrencyOverrides {maxCount=1}
 modulationSettings {pitchModulationDestination {bEnableModulation=true, modulators [null]}}
}
