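# Stereo MetaSound for the third-person SevenSlide slow-motion loop, using melee weapon sound class and concurrency with dash attenuation.
# On play, a Wave Player starts the slow-mo dash wave at a random start time.
# An audio ADSR with input-exposed attack and release times shapes both channels' volume.
# A second float ADSR, converted from frequency multiplier to semitones, sweeps the pitch over time.
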MetaSoundSource /SevenSlide/Assets/Audio/MSS_SevenSlide_Slowmo_LP_3P
inputs [
 Attack Time {type=Time}
 Release Time {type=Time}
 UE.Source.OnPlay {type=Trigger}
 Wave Asset {type=WaveAsset}
]
interfaces [UE.OutputFormat.Stereo:1.0, UE.Source.OneShot:1.0, UE.Source:1.1]
outputs [
 "UE.OutputFormat.Stereo.Audio:0" {type=Audio}
 "UE.OutputFormat.Stereo.Audio:1" {type=Audio}
 UE.Source.OneShot.OnFinished {type=Trigger}
]
pages [Default {
  input_defaults {Attack Time=0.676, Release Time=0.568, UE.Source.OnPlay=null, Wave Asset=/SevenSlide/Assets/Audio/Waves/SevenSlide_Dash_Slowmo_LP_01}
  nodes [
   ADSREnvelopeADSREnvelope_0 = ADSR Envelope.ADSR Envelope:Audio(Attack Time=^Attack Time, Release Time=^Release Time, Trigger Attack=^UE.Source.OnPlay) {also {defaults {Decay Time=1, Sustain Level=1}}}
   ADSREnvelopeADSREnvelope_1 = ADSR Envelope.ADSR Envelope:Float(Trigger Attack=^UE.Source.OnPlay) {also {defaults {Attack Time=0.4, Decay Time=7, Hard Reset="<MetaSoundBuilderResult.SUCCEEDED: 0>"}}}
   FrequencyMultipliertoSemitone_0 = UE.Frequency Multiplier to Semitone:Float(Frequency Multiplier=@ADSREnvelopeADSREnvelope_1.Out Envelope)
   WavePlayer_0 = UE.Wave Player:Stereo(
    Play=^UE.Source.OnPlay
    Wave Asset=^Wave Asset
    Pitch Shift=@FrequencyMultipliertoSemitone_0.Semitones
    Start Time=&MSP_WaveGetRandomStartTime(Sound=^Wave Asset).Value
   ) {also {defaults {Loop="<MetaSoundBuilderResult.SUCCEEDED: 0>"}, version [1,1]}}
   Multiply_0 = UE.Multiply:Audio(AdditionalOperands=@ADSREnvelopeADSREnvelope_0.Out Envelope, PrimaryOperand=@WavePlayer_0.Out Left)
   Multiply_1 = UE.Multiply:Audio(AdditionalOperands=@ADSREnvelopeADSREnvelope_0.Out Envelope, PrimaryOperand=@WavePlayer_0.Out Right)
  ]
  links [
   {dst="@UE.OutputFormat.Stereo.Audio:0", src=Multiply_0.Out}
   {dst="@UE.OutputFormat.Stereo.Audio:1", src=Multiply_1.Out}
   {dst="@UE.Source.OneShot.OnFinished", src=WavePlayer_0.On Finished}
  ]
 }]
properties {
 assetImportData=&self:AssetImportData
 attenuationSettings=&/ATT_SevenSlide_Dash
 maxDistance=4400
 numChannels=2
 outputFormat=Stereo
 sampleRate=48000
 soundClassObject=&Fort_WeaponMelee
 volume=0.354
 concurrencySet [&Player_Melee_Weapon_Conc]
}
