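# Sound cue for the landing foley of the SevenSlide feature, routed to the player foley sound class.
# An IsLocal branch selects the first-person footstep attenuation for the local player, or the SevenSlide dash attenuation when false or unset.
# Both attenuation paths share one equal-weight random pick from six waves: five foley lands and one drop land.
# A shared player foley concurrency set is assigned to limit overlapping voices.
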SoundCue /SevenSlide/Assets/Audio/SC_SevenSlide_Foley_Land
root=Branch_0
nodes [
 {id=Attenuation_0, children [Random_0], properties {attenuationSettings=&/ATT_SevenSlide_Dash}}
 {id=Attenuation_1, children [Random_0], properties {attenuationSettings=/Game/Sounds/Fort_Player/Footsteps/Settings/Fort_Player_FS_1P_Att}}
 {id=Branch_0, children [Attenuation_1, Attenuation_0, Attenuation_0], properties {boolParameterName=IsLocal}}
 {id=Random_0, children [WavePlayer_0,WavePlayer_1,WavePlayer_2,WavePlayer_3,WavePlayer_7,WavePlayer_8], properties {weights [1,1,1,1,1,1]}}
 {id=WavePlayer_0, properties {soundWaveAssetPtr=/SevenSlide/Assets/Audio/Wavs/SevenSlide_Drop_Land_01}}
 {id=WavePlayer_1, properties {soundWaveAssetPtr=/SevenSlide/Assets/Audio/Wavs/SevenSlide_Foley_Land_01}}
 {id=WavePlayer_2, properties {soundWaveAssetPtr=/SevenSlide/Assets/Audio/Wavs/SevenSlide_Foley_Land_02}}
 {id=WavePlayer_3, properties {soundWaveAssetPtr=/SevenSlide/Assets/Audio/Wavs/SevenSlide_Foley_Land_03}}
 {id=WavePlayer_7, properties {soundWaveAssetPtr=/SevenSlide/Assets/Audio/Wavs/SevenSlide_Foley_Land_04}}
 {id=WavePlayer_8, properties {soundWaveAssetPtr=/SevenSlide/Assets/Audio/Wavs/SevenSlide_Foley_Land_05}}
]
properties {
 maxDistance=4400
 soundClassObject=&Fort_Player_Foley
 volumeMultiplier=0.7
 attenuationOverrides {distanceAlgorithm=NaturalSound, falloffDistance=3000, attenuationShapeExtents {x=1000}}
 concurrencyOverrides {maxCount=3, retriggerTime=1}
 concurrencySet [/Game/Sounds/Fort_Player/Clothing/Settings/Fort_Player_Foley_Global_MaxFour]
}
