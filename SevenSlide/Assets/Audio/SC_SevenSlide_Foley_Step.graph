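# Sound Cue for SevenSlide step foley, routed to the Fort player foley sound class.
# An equally weighted Random node picks one of twelve step waves and feeds two Attenuation nodes.
# A Branch on the IsLocal parameter selects the first-person footstep attenuation for the local player, or the SevenSlide dash attenuation when false or unset.
# A shared player foley concurrency set limits simultaneous voices.
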SoundCue /SevenSlide/Assets/Audio/SC_SevenSlide_Foley_Step
root=Branch_0
nodes [
 {id=Attenuation_0, children [Random_0], properties {attenuationSettings=&/ATT_SevenSlide_Dash}}
 {id=Attenuation_1, children [Random_0], properties {attenuationSettings=/Game/Sounds/Fort_Player/Footsteps/Settings/Fort_Player_FS_1P_Att}}
 {id=Branch_0, children [Attenuation_1, Attenuation_0, Attenuation_0], properties {boolParameterName=IsLocal}}
 {
  id=Random_0
  children [
   WavePlayer_4
   WavePlayer_5
   WavePlayer_6
   WavePlayer_9
   WavePlayer_10
   WavePlayer_11
   WavePlayer_12
   WavePlayer_13
   WavePlayer_14
   WavePlayer_15
   WavePlayer_16
   WavePlayer_17
  ]
  properties {weights [1,1,1,1,1,1,1,1,1,1,1,1]}
 }
 {id=WavePlayer_10, properties {soundWaveAssetPtr=/SevenSlide/Assets/Audio/Wavs/SevenSlide_Foley_Step_09}}
 {id=WavePlayer_11, properties {soundWaveAssetPtr=/SevenSlide/Assets/Audio/Wavs/SevenSlide_Foley_Step_08}}
 {id=WavePlayer_12, properties {soundWaveAssetPtr=/SevenSlide/Assets/Audio/Wavs/SevenSlide_Foley_Step_07}}
 {id=WavePlayer_13, properties {soundWaveAssetPtr=/SevenSlide/Assets/Audio/Wavs/SevenSlide_Foley_Step_06}}
 {id=WavePlayer_14, properties {soundWaveAssetPtr=/SevenSlide/Assets/Audio/Wavs/SevenSlide_Foley_Step_05}}
 {id=WavePlayer_15, properties {soundWaveAssetPtr=/SevenSlide/Assets/Audio/Wavs/SevenSlide_Foley_Step_04}}
 {id=WavePlayer_16, properties {soundWaveAssetPtr=/SevenSlide/Assets/Audio/Wavs/SevenSlide_Foley_Step_03}}
 {id=WavePlayer_17, properties {soundWaveAssetPtr=/SevenSlide/Assets/Audio/Wavs/SevenSlide_Foley_Step_02}}
 {id=WavePlayer_4, properties {soundWaveAssetPtr=/SevenSlide/Assets/Audio/Wavs/SevenSlide_Foley_Step_01}}
 {id=WavePlayer_5, properties {soundWaveAssetPtr=/SevenSlide/Assets/Audio/Wavs/SevenSlide_Foley_Step_12}}
 {id=WavePlayer_6, properties {soundWaveAssetPtr=/SevenSlide/Assets/Audio/Wavs/SevenSlide_Foley_Step_11}}
 {id=WavePlayer_9, properties {soundWaveAssetPtr=/SevenSlide/Assets/Audio/Wavs/SevenSlide_Foley_Step_10}}
]
properties {
 maxDistance=4400
 soundClassObject=&Fort_Player_Foley
 volumeMultiplier=0.7
 attenuationOverrides {distanceAlgorithm=NaturalSound, falloffDistance=3000, attenuationShapeExtents {x=1000}}
 concurrencyOverrides {maxCount=3, retriggerTime=1}
 concurrencySet [/Game/Sounds/Fort_Player/Clothing/Settings/Fort_Player_Foley_Global_MaxFour]
}
