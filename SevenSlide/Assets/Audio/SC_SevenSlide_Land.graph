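# SevenSlide landing SoundCue, playing a wings landing sound under the in-air gliders sound class.
# A Random node picks one of three equally weighted landing waves and feeds a Modulator that varies pitch and volume slightly.
# The cue references a shared SevenSlide movement attenuation asset and overrides concurrency to cap simultaneous instances.
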SoundCue /SevenSlide/Assets/Audio/SC_SevenSlide_Land
root=Modulator_1
nodes [
 {id=Modulator_1, children [Random_1], properties {pitchMax=1, volumeMax=1.1, volumeMin=1}}
 {id=Random_1, children [WavePlayer_4,WavePlayer_5,WavePlayer_6], properties {weights [1,1,1]}}
 {id=WavePlayer_4, properties {soundWaveAssetPtr=/SevenSlide/Assets/Audio/Wavs/SevenSlide_Wings_Land_01}}
 {id=WavePlayer_5, properties {soundWaveAssetPtr=/SevenSlide/Assets/Audio/Wavs/SevenSlide_Wings_Land_02}}
 {id=WavePlayer_6, properties {soundWaveAssetPtr=/SevenSlide/Assets/Audio/Wavs/SevenSlide_Wings_Land_03}}
]
properties {
 attenuationSettings=&/Att_SevenSlide_Movement2
 bOverrideConcurrency=true
 maxDistance=12000
 soundClassObject=&Fort_Gliders_InAir
 volumeMultiplier=0.7
 attenuationOverrides {distanceAlgorithm=NaturalSound, falloffDistance=3000, attenuationShapeExtents {x=1000}}
 concurrencyOverrides {maxCount=3, retriggerTime=1}
 concurrencySet [null]
}
