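# Sound cue named for the SevenSlide launch windup, with no root or sound nodes in its graph, so it holds no audio to play.
# It keeps playback settings: the gadgets sound class, reduced volume and an attenuation asset named for third-person in-flight air.
# Concurrency references the global and local weapon gunfire sets, alongside stored single-instance, per-owner override values that are not switched on.
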SoundCue /SevenSlide/Assets/Audio/SC_SevenSlide_Launch_Windup_Cue
root=null
nodes []
properties {
 attenuationSettings=&/SevenSlide_InFlightAir_3P_Att
 maxDistance=2097152
 soundClassObject=&Fort_Gadgets
 volumeMultiplier=0.7
 concurrencyOverrides {bLimitToOwner=true, maxCount=1}
 concurrencySet [&Weapon_Gunfire_Global_Conc, &Weapon_Gunfire_Local_Conc]
}
