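# SC_SevenSlide_Rocket_Launch is a Sound Cue named for a SevenSlide rocket launch, but it has no nodes or root, so it plays no audio as it stands.
# It is assigned to the Fort_Gliders_InAir sound class and references the shared Att_SevenSlide_Movement attenuation asset.
# It also stores natural-sound attenuation override values and overrides concurrency to cap simultaneous instances.
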SoundCue /SevenSlide/Assets/Audio/SC_SevenSlide_Rocket_Launch
root=null
nodes []
properties {
 attenuationSettings=&/Att_SevenSlide_Movement
 bOverrideConcurrency=true
 maxDistance=12000
 soundClassObject=&Fort_Gliders_InAir
 volumeMultiplier=1
 attenuationOverrides {distanceAlgorithm=NaturalSound, falloffDistance=3000, attenuationShapeExtents {x=1000}}
 concurrencyOverrides {maxCount=3, retriggerTime=1}
 concurrencySet [null]
}
