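# Looping ambient-style Sound Cue named as the SevenSlide water loop, though it plays the SteamVent loop wave.
# A looping Wave Player feeds a Modulator that slightly randomizes volume, with the cue's overall volume scaled down.
# It references the SevenSlide Movement2 attenuation asset and routes through the Fortnite in-air glider sound class.
# A concurrency override limits simultaneous instances and sets a retrigger time.
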SoundCue /SevenSlide/Assets/Audio/SC_SevenSlide_WaterLoop
root=Modulator_1
nodes [
 {id=Modulator_1, children [WavePlayer_0], properties {pitchMax=1, volumeMax=1.1, volumeMin=1}}
 {id=WavePlayer_0, properties {bLooping=true, soundWaveAssetPtr=/SevenSlide/Assets/Audio/Wavs/SevenSlide_SteamVent_Loop}}
]
properties {
 attenuationSettings=&/Att_SevenSlide_Movement2
 bOverrideConcurrency=true
 maxDistance=12000
 soundClassObject=&Fort_Gliders_InAir
 volumeMultiplier=0.35
 attenuationOverrides {distanceAlgorithm=NaturalSound, falloffDistance=3000, attenuationShapeExtents {x=1000}}
 concurrencyOverrides {maxCount=3, retriggerTime=1}
 concurrencySet [null]
}
