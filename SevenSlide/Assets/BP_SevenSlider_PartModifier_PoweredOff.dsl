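; Character part modifier for the Seven Slider that drives its mesh materials from gameplay tags on the owning player pawn.
; On post-res-in, it chains tag add and remove wait tasks through their immediate outputs, registering all listeners together.
; PoweredOn toggles the boot animation parameter with game-time stamps for shader timing, while Overheat toggles its own parameter and timestamps only removal.
; The activewindow tag only prints debug messages.
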(event EventOnPostResIn
  (bind _returnvalue (CustomCharacter|GetAssociatedPlayerPawn))
  (bind _returnvalue_2 (CustomCharacter|GetSkeletalMeshForPartOnPlayerPawn))
  (bind _returnvalue_3 (Utilities|Time|GetGameTimeinSeconds))
  (bind _returnvalue_4 (CustomCharacter|GetSkeletalMeshForPartOnPlayerPawn))
  (|CallParentFunction)
  (Ability|Tasks|WaitGameplayTagAddToActor _returnvalue "(TagName=\"sevenslide.PoweredOn\")"
    (:then
      (Ability|Tasks|WaitGameplayTagRemoveFromActor _returnvalue "(TagName=\"sevenslide.PoweredOn\")"
        (:then
          (Ability|Tasks|WaitGameplayTagAddToActor _returnvalue "(TagName=\"sevenslide.Overheat\")"
            (:then
              (Ability|Tasks|WaitGameplayTagRemoveFromActor _returnvalue "(TagName=\"sevenslide.Overheat\")"
                (:then
                  (Ability|Tasks|WaitGameplayTagAddToActor _returnvalue "(TagName=\"sevenslide.activewindow\")"
                    (:then
                      (Ability|Tasks|WaitGameplayTagRemoveFromActor _returnvalue "(TagName=\"sevenslide.activewindow\")"
                        (:Removed
                          (Development|PrintString "OUT"))))
                    (:Added
                      (Development|PrintString "ActiveWindowwww"))))
                (:Removed
                  (Rendering|Material|SetScalarParameterValueonMaterials _returnvalue_4 "BP_Overheat" 0.0)
                  (Rendering|Material|SetScalarParameterValueonMaterials _returnvalue_4 "BP_GetTime_Overheat" (Utilities|Time|GetGameTimeinSeconds)))))
            (:Added
              (Rendering|Material|SetScalarParameterValueonMaterials _returnvalue_4 "BP_Overheat" 1.0))))
        (:Removed
          (Rendering|Material|SetScalarParameterValueonMaterials _returnvalue_2 "BP_GetTime_Deactivate" _returnvalue_3)
          (Rendering|Material|SetScalarParameterValueonMaterials _returnvalue_2 "BP_ActivateBootAnimationBP" 0.0))))
    (:Added
      (Rendering|Material|SetScalarParameterValueonMaterials _returnvalue_2 "BP_GetTime_Activate" _returnvalue_3)
      (Rendering|Material|SetScalarParameterValueonMaterials _returnvalue_2 "BP_ActivateBootAnimationBP" 1.0))))

(fn ConstructionScript
  (|CallParentFunction))
