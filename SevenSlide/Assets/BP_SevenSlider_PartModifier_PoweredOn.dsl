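; BP_SevenSlider_PartModifier_PoweredOn is a SevenSlider part modifier that drives powered-on and overheat material effects on the player's part skeletal mesh.
; On post res-in, it stamps game time into the activation material parameter.
; Two gameplay-tag tasks, started together, watch the pawn for sevenslide.Overheat.
; Tag addition enables the overheat parameters; removal clears BP_Overheat and stamps its time, leaving BP_AllowOverheat set.
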(event EventOnPostResIn
  (bind _returnvalue (CustomCharacter|GetAssociatedPlayerPawn))
  (bind _returnvalue_2 (CustomCharacter|GetSkeletalMeshForPartOnPlayerPawn))
  (|CallParentFunction)
  (Utilities|FlowControl|Sequence
    (:then_0
      (Rendering|Material|SetScalarParameterValueonMaterials (CustomCharacter|GetSkeletalMeshForPartOnPlayerPawn) "BP_GetTime_Activate" (Utilities|Time|GetGameTimeinSeconds)))
    (:then_1
      (Ability|Tasks|WaitGameplayTagAddToActor _returnvalue "(TagName=\"sevenslide.Overheat\")"
        (:then
          (Ability|Tasks|WaitGameplayTagRemoveFromActor _returnvalue "(TagName=\"sevenslide.Overheat\")"
            (:Removed
              (Rendering|Material|SetScalarParameterValueonMaterials _returnvalue_2 "BP_Overheat" 0.0)
              (Rendering|Material|SetScalarParameterValueonMaterials _returnvalue_2 "BP_GetTime_Overheat" (Utilities|Time|GetGameTimeinSeconds)))))
        (:Added
          (Rendering|Material|SetScalarParameterValueonMaterials _returnvalue_2 "BP_Overheat" 1.0)
          (Rendering|Material|SetScalarParameterValueonMaterials _returnvalue_2 "BP_AllowOverheat" 1.0))))))

(fn ConstructionScript
  (|CallParentFunction))
