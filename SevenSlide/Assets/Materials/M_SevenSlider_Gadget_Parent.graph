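# Parent material for the SevenSlider gadget, wrapping the 2023 character pre/post-FX functions with overheat and activation effects.
# Overheat blends pulsing, gradient-colored emissive and base color using vertex-color propeller masks and sphere-mask boot hotspots.
# Activation drives a hexagon-texture opacity dissolve with blue emissive, with vertex color masking the ankle.
# Both effects time from Blueprint-set timestamp parameters compared against Time, and pre-skinned normals drive world position offset.
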Material /SevenSlide/Assets/Materials/M_SevenSlider_Gadget_Parent
graph {
 properties {
  bAutomaticallySetUsageInEditor=false
  bUseMaterialAttributes=true
  bUsedWithMeshDeformer=true
  bUsedWithMorphTargets=true
  bUsedWithSkeletalMesh=true
  pixelDepthOffsetMode=PDOM_Legacy
  shadingModel=MSM_FromMaterialExpression
  twoSided=true
  usedShadingModels=Default Lit | Subsurface | Clear Coat | Subsurface Profile | Cloth
 }
 nodes [
  Constant() {r=1}
  n1 = MaterialFunctionCall() {materialFunction=&MF_CharacterParent_2023_PreFX}
  MaterialFunctionCall(Material=NamedRerouteUsage(), M=@n1.M, Pre Skinned Local Position=@n1.M, Distance Blend=@n1.M, Fresnel=@n1.M) {materialFunction=&MF_CharacterParent_2023_PostFX}
  NamedRerouteDeclaration(None=@n1.Pre skinned Local Position) {name=PS_LocalPosition, nodeColor=<rgba 0 1 0.6719 1>}
  NamedRerouteDeclaration(None=@n1.Result) {name=InMA, nodeColor=<rgba 1 0.3516 0 1>}
  n2 = NamedRerouteUsage()
  n3 = Multiply(A=@n2) {constB=0.01}
  n4 = Multiply(A=@n2) {constB=0.12}
  n5 = Time() {bOverride_Period=true, period=10}
  n6 = ConstantBiasScale(None=Sine(None=Multiply(A=@n5) {constB=1.1}))
  n7 = NamedRerouteUsage()
  n8 = NamedRerouteUsage()
  n9 = ScalarParameter() {parameterName=Overheat_MettalicInfluence}
  n10 = Multiply(A=Add(A=NamedRerouteUsage(), B=OneMinus(None=@n9)) {constB=0}, B=NamedRerouteUsage())
  n11 = NamedRerouteDeclaration(None=Saturate(None=Add(A=@n8, B=@n10))) {name=PropellerMask_Clean, nodeColor=<rgba 0 1 0.3203 1>}
  n12 = Add(A=ComponentMask(None=NamedRerouteUsage()) {g=true}) {constB=10}
  n13 = OneMinus(None=Saturate(None=Divide(A=@n12) {constB=12}))
  n14 = OneMinus(None=ComponentMask(None=NamedRerouteUsage()) {r=true})
  n15 = ScalarParameter() {defaultValue=10, parameterName=Overheat_AO_Influence}
  n16 = Multiply(A=Constant3Vector() {constant=<rgba 1 0.6395 0.0166 1>}, B=Add(A=Power(Base=@n7) {constExponent=0.6}, B=Multiply(A=@n14, B=@n15)))
  n17 = Constant3Vector() {constant=<rgba 0.7483 0.0742 0 1>}
  n18 = Constant3Vector() {constant=<rgba 0.7413 0 0.017 1>}
  n19 = NamedRerouteUsage()
  n20 = NamedRerouteDeclaration(None=SphereMask(A=ComponentMask(None=@n19) {g=true}, B=Constant() {r=-14}, Radius=ScalarParameter() {defaultValue=50, parameterName=OverheatRadius}) {attenuationRadius=10, hardnessPercent=0}) {name=Y-Axis_OverheatGradient, nodeColor=<rgba 1 0 0.75 1>}
  n21 = NamedRerouteDeclaration(None=Add(A=Multiply(A=@n16) {constB=0.15}, B=LinearInterpolate(A=@n17, B=@n18, Alpha=@n20))) {name=OverheatColors, nodeColor=<rgba 1 0 0 1>}
  n22 = Power(Base=NamedRerouteUsage()) {constExponent=0.2}
  n23 = Multiply(A=Add(A=Power(Base=@n7)) {constB=0.3}, B=Multiply(A=Multiply(A=@n11, B=@n13), B=Multiply(A=@n21, B=@n22)))
  NamedRerouteDeclaration(None=Add(A=LinearInterpolate(A=@n3, B=@n4, Alpha=@n6), B=Multiply(A=@n23) {constB=1.5})) {name=OverheatEmissive, nodeColor=<rgba 0.6406 0 1 1>}
  n24 = Subtract(A=Time(), B=ScalarParameter() {defaultValue=100000000, parameterName=BP_GetTime_Overheat})
  n25 = OneMinus(None=Saturate(None=Multiply(A=@n24) {constB=2}))
  n26 = ScalarParameter() {parameterName=BP_Overheat}
  NamedRerouteDeclaration(None=Multiply(A=LinearInterpolate(A=@n25, Alpha=@n26), B=ScalarParameter() {parameterName=BP_AllowOverheat})) {name=OverheatTiming, nodeColor=<rgba 1 0 0 1>}
  n27 = GetMaterialAttributes(None=NamedRerouteUsage()) {attributeGetTypes [
    #1
    #2
    #3
    #4
   ]}
  NamedRerouteDeclaration(None=@n27.Roughness) {name=OriginalRoughness, nodeColor=<rgba 0 0.1094 1 1>}
  NamedRerouteDeclaration(None=@n1.Fresnel) {name=InFresnel, nodeColor=<rgba 1 0.9609 0 1>}
  NamedRerouteDeclaration(None=@n1.M) {name=InMasks, nodeColor=<rgba 1 0.9375 0 1>}
  n28 = NamedRerouteUsage()
  n29 = Constant3Vector() {constant=<rgba -24 -12 40 0>}
  n30 = Multiply(A=Constant3Vector() {constant=<rgba -1 1 1 0>}, B=@n29)
  n31 = ScalarParameter() {defaultValue=16, parameterName=asfdasdgdfb}
  n32 = Constant3Vector() {constant=<rgba -16 -3 2 0>}
  n33 = Multiply(A=@n32, B=Constant3Vector() {constant=<rgba -1 1 1 0>})
  n34 = Add(
   A=Add(
    A=SphereMask(A=@n28, B=@n30, Radius=@n31) {attenuationRadius=15, hardnessPercent=0}
    B=SphereMask(A=@n28, B=@n29, Radius=@n31) {attenuationRadius=15, hardnessPercent=0}
   )
   B=Add(
    A=SphereMask(A=@n28, B=@n32) {attenuationRadius=15, hardnessPercent=0}
    B=SphereMask(A=@n28, B=@n33) {attenuationRadius=15, hardnessPercent=0}
   )
  )
  NamedRerouteDeclaration(None=Multiply(A=@n34, B=@n34)) {name=AdditionalBootOverheatMask, nodeColor=<rgba 0 1 0.1328 1>}
  n35 = LinearInterpolate(A=NamedRerouteUsage(), B=NamedRerouteUsage(), Alpha=NamedRerouteUsage())
  n36 = Step(X=NamedRerouteUsage()) {constY=0.5}
  n37 = LinearInterpolate(B=NamedRerouteUsage(), Alpha=Step(X=NamedRerouteUsage()) {constY=0.5}) {constA=0.05}
  n38 = Constant3Vector() {constant=<rgba 0.8212 0.0755 0 1>}
  n39 = NamedRerouteUsage()
  n40 = NamedRerouteUsage()
  n41 = Add(A=Multiply(A=Multiply(A=@n39, B=@n40)) {constB=0.5}, B=Multiply(A=NamedRerouteUsage()) {constB=0.75})
  n42 = PreSkinnedNormal()
  n43 = ShaderStageSwitch(PixelShader=VertexInterpolator(VS=@n42).PS, VertexShader=@n42)
  n44 = Multiply(A=NamedRerouteUsage()) {constB=1.5}
  n45 = SetMaterialAttributes!(
   MaterialAttributes=@n27.MaterialAttributes
   Opacity Mask
   Emissive Color=Add(A=Multiply(A=@n35, B=@n36), B=NamedRerouteUsage())
   Base Color=LinearInterpolate(A=@n37, B=LinearInterpolate(A=@n37, B=@n38, Alpha=@n41), Alpha=NamedRerouteUsage())
   World Position Offset=Transform(None=Multiply(A=@n43, B=@n44) {constB=100}) {transformSourceType=TRANSFORMSOURCE_Local}
  ) {attributeSetTypes [
    #5
    #1
    #3
    #6
   ]}
  NamedRerouteDeclaration(None=@n45) {name=OutMA, nodeColor=<rgba 0.0312 0 1 1>}
  n46 = TextureSample() {samplerSource=SSM_Wrap_WorldGroupSettings, samplerType=SAMPLERTYPE_Masks, texture=/SevenSlide/Assets/Textures/T_SevenSlider_TA_Hexagon}
  NamedRerouteDeclaration(None=@n46.G) {name=OverheatEdgesMask, nodeColor=<rgba 0.0781 0 1 1>}
  n47 = VertexColor()
  NamedRerouteDeclaration(None=Ceil(None=@n47.R)) {name=AnkleMask, nodeColor=<rgba 1 0 0.2286 1>}
  n48 = NamedRerouteUsage()
  n49 = Subtract(A=Subtract(A=Time() {period=10}, B=ScalarParameter() {defaultValue=1000000000, parameterName=BP_GetTime_Activate})) {constB=-0.1}
  n50 = OneMinus(None=Saturate(None=Multiply(A=@n49) {constB=0.65}))
  n51 = Multiply(A=@n50, B=@n50)
  n52 = OneMinus(None=Multiply(A=@n51, B=@n51))
  n53 = Add(A=Add(A=@n46.R) {constB=0.3}, B=@n52)
  n54 = ComponentMask(None=NamedRerouteUsage()) {b=true}
  n55 = OneMinus(None=Saturate(None=Divide(A=@n54) {constB=100}))
  n56 = Saturate(None=Multiply(A=Multiply(A=@n53, B=@n55), B=@n52))
  NamedRerouteDeclaration(None=Saturate(None=Add(A=@n48, B=@n56))) {name=On/Off_OpacityMask, nodeColor=<rgba 0.5234 1 0 1>}
  n57 = Constant() {r=0.5}
  n58 = Multiply(
   A=OneMinus(None=@n48)
   B=NamedRerouteDeclaration(None=SphereMask(A=@n56, B=@n57) {attenuationRadius=0.1, hardnessPercent=0}) {name=ActiveHexagonsMask, nodeColor=<rgba 0.1484 0 1 1>}
  )
  n59 = Constant3Vector() {constant=<rgba 0 0.3892 1 1>}
  n60 = Multiply(A=@n59) {constB=20}
  n61 = Add(A=Subtract(A=@n52)) {constB=0.5}
  n62 = Multiply(A=Saturate(None=Sine(None=@n61)), B=@n59)
  NamedRerouteDeclaration(None=Add(A=Multiply(A=@n58, B=@n60), B=Multiply(A=@n62) {constB=0.2})) {name=On/Off_EmissiveTransition, nodeColor=<rgba 1 0 0.8437 1>}
  n63 = Ceil(None=@n47.G)
  NamedRerouteDeclaration(None=@n63) {name=PropellerMask, nodeColor=<rgba 0 1 0.2502 1>}
  NamedRerouteDeclaration(None=@n27.Emissive Color) {name=OriginalEmissive, nodeColor=<rgba 1 0 0.2109 1>}
  NamedRerouteDeclaration(None=@n27.Metallic) {name=OriginalMetallic, nodeColor=<rgba 0 1 0.0156 1>}
  NamedRerouteDeclaration(None=@n27.Base Color) {name=OriginalBaseColor, nodeColor=<rgba 1 0.5939 0 1>}
  OneMinus(None=@n63)
  OneMinus()
  Saturate()
 ]
}
