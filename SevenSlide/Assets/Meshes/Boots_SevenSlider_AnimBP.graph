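# Animation Blueprint for the SevenSlide boots on the SevenSlider gadget skeleton, layering slide lean, slow-motion and overheat poses onto the wearer's animation.
# It copies and caches the attached parent mesh's pose, then applies an additive LeanAmount blendspace while sliding.
# Slow motion swaps in an additive blendspace driven by OppositeLeanAmount, and overheating overrides both with an additive pose, using eased bool blends.
# The EventGraph is empty and the driving variables are not declared here.
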AnimBlueprint /SevenSlide/Assets/Meshes/Boots_SevenSlider_AnimBP
target_skeleton=&/SevenSlider_Gadget_Skeleton
graphs [
 AnimGraph {nodes [
   n1 = "Sequence Player (additive) 'SevenSlider_Gadget_OverheatPose'"() {sequence=/SevenSlide/Assets/Animation/Gadget/SevenSlider_Gadget_OverheatPose}
   n2 = Use cached pose 'BaseAnim_Cache'()
   n3 = Apply Additive(
    Base=Use cached pose 'BaseAnim_Cache'()
    Additive=Blendspace Player 'SevenSlider_Gadget_Slomo_Lean_Additive_BS'!(X=Get OppositeLeanAmount() -> (OppositeLeanAmount:double)) -> (Pose) {blendSpace=/SevenSlide/Assets/Animation/Gadget/Slomo/SevenSlider_Gadget_Slomo_Lean_Additive_BS}
   )
   n4 = Use cached pose 'Lean_Cache'()
   n5 = Get bIsSlowMo() -> (bIsSlowMo:bool)
   Output Pose(Result=Blend Poses by bool(
     BlendPose_0=Apply Additive(Base=@n2, Additive=@n1)
     BlendPose_1=Blend Poses by bool(BlendPose_0=@n3, BlendPose_1=@n4, BlendTime_0=0.3, BlendTime_1=0.3, bActiveValue=@n5) {blendType=CubicInOut}
     BlendTime_0=0.3
     BlendTime_1=0.3
     bActiveValue=Get bIsOverheated() -> (bIsOverheated:bool)
    ) {blendType=CubicInOut})
   Save cached pose 'BaseAnim_Cache'(Pose=Copy Pose From Mesh() {bUseAttachedParent=true}) {cacheName=BaseAnim_Cache}
   n6 = Blendspace Player 'SevenSlider_Gadget_Lean_Additive_BS'!(X=Get LeanAmount() -> (LeanAmount:double)) -> (Pose) {blendSpace=/SevenSlide/Assets/Animation/Gadget/SevenSlider_Gadget_Lean_Additive_BS}
   n7 = Use cached pose 'BaseAnim_Cache'()
   Save cached pose 'Lean_Cache'(Pose=Blend Poses by bool(
     BlendPose_0=Apply Additive(Base=@n7, Additive=@n6)
     BlendPose_1=Use cached pose 'BaseAnim_Cache'()
     BlendTime_0=0.3
     BlendTime_1=0.3
     bActiveValue=Get bSliderIsSliding() -> (bSliderIsSliding:bool)
    ) {blendType=CubicInOut}) {cacheName=Lean_Cache}
  ]}
 EventGraph {links [], nodes []}
]
variables []
